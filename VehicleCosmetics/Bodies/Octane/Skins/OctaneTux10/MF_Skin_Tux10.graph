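# Material function for the Octane Tux10 body skin, outputting material attributes for a dark navy and pale gold finish.
# The skin texture's blue channel masks navy from gold and drives metallic, roughness and a matte normal that perturbs only the Fresnel input.
# Fresnel shifts edges toward gold, shades color and adds emissive rim glow.
# Masked gold areas gain faked highlights from a tangent-space camera dot product and a reflection-vector environment texture lookup.
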MaterialFunction /VehicleCosmetics/Bodies/Octane/Skins/OctaneTux10/MF_Skin_Tux10
graph {nodes [
  n1 = Constant3Vector() {constant=<rgba 0 0 1 0>}
  ConstantBiasScale(None=@n1) {bias=-0.5, scale=2}
  Fresnel() {baseReflectFraction=0, exponent=3}
  n2 = Constant3Vector() {constant=<rgba 0.0073 0.0073 0.0696 0>}
  n3 = Constant3Vector() {constant=<rgba 0.75 0.625 0.35 1>}
  n4 = TextureSample() {samplerType=SAMPLERTYPE_LinearColor, texture=/VehicleCosmetics/Bodies/Octane/Skins/skin_octane_TuxBW_01_RGB}
  n5 = TextureSample(UVs=TextureCoordinate() {coordinateIndex=1, uTiling=3, vTiling=3}) {constCoordinate=1, samplerType=SAMPLERTYPE_Normal, texture=/VehicleCosmetics/SharedTextures/T_Matte_N}
  n6 = MaterialFunctionCall!(Normal=LinearInterpolate(A=LinearInterpolate(A=@n1, B=@n5.RGB), B=@n1, Alpha=@n4.B), Exponent=Constant() {r=3}) {materialFunction=/VehicleCosmetics/Wheels/Cola_Holographic/Materials/MF_Fresnel_Classic}
  n7 = ConstantBiasScale(None=@n6.Result) {bias=-0.01, scale=1.2}
  n8 = ConstantBiasScale(None=@n6.Result) {bias=-0.5, scale=8}
  n9 = Fresnel() {baseReflectFraction=0, exponent=0.5}
  n10 = Clamp(None=ConstantBiasScale(None=OneMinus(None=@n9)) {bias=-0.125, scale=3}) {maxDefault=0.8, minDefault=0.3}
  n11 = Multiply(A=LinearInterpolate(A=LinearInterpolate(A=@n2, B=@n3, Alpha=@n4.B), B=@n3, Alpha=Saturate(None=@n7)), B=Saturate(None=Max(A=@n8, B=@n10)))
  n12 = Transform(None=CameraVectorWS()) {transformSourceType=TRANSFORMSOURCE_World, transformType=TRANSFORM_Tangent}
  n13 = Constant3Vector() {constant=<rgba -0.1 1.9 1 0>}
  n14 = Power(Base=ConstantBiasScale(None=DotProduct(A=@n12, B=@n13)) {bias=-0.9, scale=1}) {constExponent=1}
  n15 = ConstantBiasScale(None=@n14) {bias=-0.5, scale=0.2}
  n16 = ConstantBiasScale(None=@n14) {bias=-0.99, scale=10}
  n17 = ReflectionVectorWS()
  n18 = TextureSample(UVs=ComponentMask(None=ConstantBiasScale(None=@n17)) {g=true, r=true}) {texture=/VehicleCosmetics/Textures/UniversalTextures/T_ENVPack}
  n19 = LinearInterpolate(A=@n18.R, B=@n18.R) {constAlpha=0.4}
  n20 = LinearInterpolate(B=Add(A=Max(A=@n15, B=@n16), B=ConstantBiasScale(None=@n19) {bias=-0.074, scale=3}), Alpha=@n4.B)
  n21 = OneMinus(None=Fresnel() {baseReflectFraction=0.5, exponent=0.1})
  n22 = ConstantBiasScale(None=@n6.Result) {bias=-0.3, scale=3}
  n23 = Add(A=Multiply(A=@n11, B=Multiply(A=@n4.B, B=@n21)), B=Multiply(A=@n3, B=Saturate(None=@n22)))
  n24 = SetMaterialAttributes!(
   MaterialAttributes
   BaseColor=Add(A=Multiply(A=@n11) {constB=0.75}, B=Multiply(A=@n11, B=@n20))
   Metallic=@n4.B
   Roughness=LinearInterpolate(Alpha=@n4.B) {constA=0.1, constB=0.4}
   EmissiveColor=Multiply(A=Saturate(None=@n23)) {constB=0.5}
   Normal
   AmbientOcclusion
  ) {attributeSetTypes [
    #1
    #2
    #3
    #4
    #5
    #6
   ]}
  FunctionOutput(None=@n24)
  LinearInterpolate()
 ]}
